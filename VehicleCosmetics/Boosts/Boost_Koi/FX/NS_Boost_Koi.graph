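# Niagara system for the Koi vehicle boost cosmetic, emitting a koi trail effect from the car's filtered boost-port sockets.
# A mesh-and-sprite koi emitter joins three ribbon trails (aura, green wave, blue wave) and small sprites swirled by vortex force.
# Each emitter picks painted-variant colors, while user flags for frontend preview and painted state retune values like velocity, spawn rate and brightness.
# Trails and sprites inherit source movement, mostly fading by curves over particle life.
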NiagaraSystem /VehicleCosmetics/Boosts/Boost_Koi/FX/NS_Boost_Koi
dependencies {
 usedDataInterfaces [&NiagaraDataInterfaceCurve, &NiagaraDataInterfaceArrayFloat, &NiagaraDataInterfaceSkeletalMesh]
 usedDynamicInputs []
 usedModules [
  &self:SystemSpawnScript.SystemScriptSource.SystemScriptGraph.NiagaraNodeAssignment_0.SetVariables_#1
  &SystemState
 ]
}
emitters {
 NE_VC_Boost_Aura {
  emitterUpdateScript [
   &EmitterState(Life Cycle Mode=::NewEnumerator1, Loop Duration=1, UseLoopDelay=false)
   &VC_FilteredSocketCount()
   &VC_SpawnRibbon_v1(Number of Ribbons=$Emitter.PortCount, SpawnRate=200)
  ]
  particleSpawnScript [
   &VC_PaintedVariantsOverride(
    Black=<rgba 0.01 0.01 0.01 2>
    BurntSienna=<rgba 0.15 0.0317 0 2>
    Crimson=<rgba 0.3 0.01 0 0.5>
    DefaultColor=&MakeCustomLinearColorFromBool(Bool=$User.bIsFrontendPreview, True Color=<rgba 0.02 1 0.9183 1>, False Color=<rgba 0.02 1 0.65 1>)
    ForestGreen=<rgba 0 0.3 0.0152 0.7>
    Grey=<rgba 0.1 0.1 0.1 1>
    Lime=<rgba 0.2667 0.4 0 1>
    Orange=<rgba 0.7 0.14 0 1>
    Pink=<rgba 0.2928 0.105 0.35 1.5>
    Purple=<rgba 0.0833 0 1 2>
    Saffron=<rgba 0.3 0.3 0 1>
    SkyBlue=<rgba 0 0.1667 0.5 1>
    TitaniumWhite=<rgba 0.5 0.5 0.5 0.5>
   )
   &VC_InitializeRibbon_v1(
    Color=$Particles.Color
    Custom Offset=true
    Enable Skeletal Position=true
    Lifetime=&MakeCustomFloatFromBool(Bool=$User.bIsFrontendPreview, True Float=0.7, False Float=0.7)
    Local Vector Offset=&AddVectorToPosition(Position=<40 0 0>, Vector=$User.DTCarOffset)
    Offset Amount=1
    Ribbon Width=120
   )
   &AddVelocity(
    Velocity=&MakeCustomVectorFromBool(Bool=$User.bIsFrontendPreview, True Vector=<4500 0 0>, False Vector=<3000 0 0>)
    Velocity Speed Scale=-1
   ) {module=AddVelocity001}
  ]
  particleUpdateScript [
   &ParticleState()
   &InheritSourceMovement(
    Applied Force Scale=&Multiply_VectorByFloat(
     Vector=<1 1 1>
     Float=&MakeCustomFloatFromBool(Bool=$User.bIsFrontendPreview, True Float=0, False Float=1)
    )
    Applied Position Scale=&Multiply_VectorByFloat(
     Vector=<1 1 1>
     Float=&MakeCustomFloatFromBool(
      Bool=$User.bIsFrontendPreview
      True Float=0
      False Float=&FloatFromCurve(FloatCurve=DataInterface() {propertyValues {Curve {keys [
           {arriveTangent=0.6707, interpMode=RCIM_Cubic, leaveTangent=0.6707, tangentMode=RCTM_User, time=0, value=1}
           {arriveTangent=6.4193, interpMode=RCIM_Cubic, leaveTangent=6.4193, tangentMode=RCTM_User, time=1, value=0}
          ]}}})
     )
    )
    ApplyForce=true
    ApplyForceFalloff=false
    ApplyPosition=true
    ApplyPositionFalloff=false
    Module.LimitForceSpeed=false
   )
   &SolveForcesAndVelocity(Clamp Velocity=false, Limit Acceleration=false)
   &ScaleColor(
    Scale Alpha=&MakeCustomFloatFromBool(
     Bool=$User.bIsFrontendPreview
     True Float=&MakeCustomFloatFromBool(Bool=$User.bIsPainted, True Float=0.4, False Float=0.35)
     False Float=&FloatFromCurve(FloatCurve=DataInterface() {propertyValues {Curve {keys [
          {arriveTangent=-3.4487, interpMode=RCIM_Cubic, leaveTangent=-3.4487, tangentMode=RCTM_User, time=0, value=1}
          {interpMode=RCIM_Cubic, tangentMode=RCTM_User, time=1, value=0}
         ]}}})
    )
    Scale RGB=&Multiply_VectorByFloat(Vector=<1 1 1>, Float=&MakeCustomFloatFromBool(Bool=$User.bIsPainted, True Float=2, False Float=1))
    ScaleA=true
    ScaleRGB=true
   )
   &VC_UpdateRibbon_v1(Enable Skeletal Position=true)
   &DynamicMaterialParameters(
    Index 0 Param 1=1
    Index 0 Param 2=1
    Index 0 Param 4=&MakeCustomFloatFromBool(
     Bool=$User.bIsFrontendPreview
     True Float=2
     False Float=&FloatFromCurve(FloatCurve=DataInterface() {propertyValues {Curve {keys [<at 0 1>,<at 1 0>]}}}, Scale Curve=3)
    )
    Module.Index0Param0Write=true
    Module.Index0Param1Write=true
    Module.Index0Param2Write=true
    Module.Index0Param3Write=true
   )
   &ScaleRibbonWidth(Ribbon Width Scale=&SelectFloatFromArray(Direct Array Index=$Particles.SpawnGroup, Float Selection Array=DataInterface() {propertyValues {FloatData [1,-1]}})) {enabled=false}
  ]
  rendererClasses [&NiagaraRibbonRendererProperties]
  renderers [{rendererClass=&NiagaraRibbonRendererProperties}]
 }
 NE_VC_Boost_BLUEWAVE001 {
  emitterUpdateScript [
   &EmitterState(
    Inactive Response=::NewEnumerator1
    Life Cycle Mode=::NewEnumerator1
    Loop Duration=0.1
    Recalculate Duration Each Loop=true
    UseLoopDelay=false
   )
   &VC_FilteredSocketCount()
   &VC_SpawnRibbon_v1(Number of Ribbons=$Emitter.PortCount, SpawnRate=400)
  ]
  particleSpawnScript [
   &VC_PaintedVariantsOverride(
    Black=<rgba 0.004 0.004 0.004 4>
    BurntSienna=<rgba 0.2 0.0539 0.0247 1.5>
    Cobalt=<rgba 0 0.0167 1 1>
    Crimson=<rgba 0.25 0 0 1.5>
    DefaultColor=&MakeCustomLinearColorFromBool(Bool=$User.bIsFrontendPreview, True Color=<rgba 0 0.5 0.375 2>, False Color=<rgba 0 0.9 0.345 1>)
    ForestGreen=<rgba 0 0.12 0 2>
    Grey=<rgba 0.07 0.07 0.07 2>
    Lime=<rgba 0.1333 0.2 0 5>
    Orange=<rgba 0.5 0.125 0 2>
    Pink=<rgba 0.38 0.1 0.5 2>
    Purple=<rgba 0.0933 0 0.8 1.2>
    Saffron=<rgba 0.25 0.2083 0 1.2>
    SkyBlue=<rgba 0 0.1 0.4 2>
    TitaniumWhite=<rgba 0.25 0.25 0.25 1.5>
   )
   &VC_InitializeRibbon_v1(
    Color=$Particles.Color
    Custom Offset=true
    Enable Skeletal Position=true
    Local Vector Offset=&AddVectorToPosition(
     Position=&MakeCustomVectorFromBool(Bool=$User.bIsFrontendPreview, True Vector=<50 0 0>, False Vector=<-75 0 0>)
     Vector=$User.DTCarOffset
    )
    Offset Amount=1
    Ribbon Width=270
   )
   &AddVelocity(
    Velocity=&MakeCustomVectorFromBool(Bool=$User.bIsFrontendPreview, True Vector=<4500 0 0>, False Vector=<2500 0 0>)
    Velocity Speed Scale=-1
   ) {module=AddVelocity001}
  ]
  particleUpdateScript [
   &ParticleState()
   &InheritSourceMovement(
    Applied Force Scale=&Multiply_VectorByFloat(
     Vector=<1 1 1>
     Float=&MakeCustomFloatFromBool(Bool=$User.bIsFrontendPreview, True Float=0, False Float=1)
    )
    Applied Position Scale=&Multiply_VectorByFloat(
     Vector=<1 1 1>
     Float=&MakeCustomFloatFromBool(
      Bool=$User.bIsFrontendPreview
      True Float=0
      False Float=&FloatFromCurve(FloatCurve=DataInterface() {propertyValues {Curve {keys [
           {arriveTangent=-1.0508, interpMode=RCIM_Cubic, leaveTangent=-1.0508, tangentMode=RCTM_User, time=0, value=1}
           {interpMode=RCIM_Cubic, tangentMode=RCTM_User, time=0.98, value=0}
          ]}}})
     )
    )
    ApplyForce=true
    ApplyForceFalloff=false
    ApplyPosition=true
    ApplyPositionFalloff=false
    Module.LimitForceSpeed=false
   )
   &Drag(Drag=0.5, Rotational Drag=10, Use Linear Drag=true, Use Rotational Drag=true)
   &SolveForcesAndVelocity(Clamp Velocity=false, Limit Acceleration=false)
   &VC_UpdateRibbon_v1(Enable Skeletal Position=true)
   &DynamicMaterialParameters(
    Index 0 Param 1=&MakeCustomFloatFromBool(Bool=$User.bIsFrontendPreview, True Float=1, False Float=1.5)
    Index 0 Param 2=1
    Index 0 Param 3=&MakeCustomFloatFromBool(
     Bool=$User.bIsFrontendPreview
     True Float=&MakeCustomFloatFromBool(Bool=$User.bIsPainted, True Float=1, False Float=1)
     False Float=&MakeCustomFloatFromBool(Bool=$User.bIsPainted, True Float=0.5, False Float=0.5)
    )
    Index 0 Param 4=&MakeCustomFloatFromBool(Bool=$User.bIsPainted, True Float=100, False Float=5)
    Module.Index0Param0Write=true
    Module.Index0Param1Write=true
    Module.Index0Param2Write=true
    Module.Index0Param3Write=true
   )
   &ScaleRibbonWidth(Ribbon Width Scale=&SelectFloatFromArray(Direct Array Index=$Particles.SpawnGroup, Float Selection Array=DataInterface() {propertyValues {FloatData [-1,1]}}))
   &DynamicMaterialParameters(
    Index 0 Param 1=&MakeCustomFloatFromBool(Bool=$User.bIsFrontendPreview, True Float=1, False Float=1.5)
    Index 0 Param 2=1
    Index 0 Param 3=&MakeCustomFloatFromBool(
     Bool=$User.bIsFrontendPreview
     True Float=&MakeCustomFloatFromBool(
      Bool=$User.bIsPainted
      True Float=0.5
      False Float=&FloatFromCurve(FloatCurve=DataInterface() {propertyValues {Curve {keys [<at 0 1.5>,<at 0.3 0.25>]}}}, Scale Curve=1.5)
     )
     False Float=&MakeCustomFloatFromBool(Bool=$User.bIsPainted, True Float=0.5, False Float=0.25)
    )
    Index 0 Param 4=&MakeCustomFloatFromBool(Bool=$User.bIsPainted, True Float=0.5, False Float=0.5)
    Module.Index0Param0Write=true
    Module.Index0Param1Write=true
    Module.Index0Param2Write=true
    Module.Index0Param3Write=true
   ) {module=DynamicMaterialParameters001}
   &ScaleColor(
    Scale Alpha=&FloatFromCurve(FloatCurve=DataInterface() {propertyValues {Curve {keys [
         {arriveTangent=1.8262, interpMode=RCIM_Cubic, leaveTangent=1.8262, tangentMode=RCTM_User, time=0, value=0}
         {interpMode=RCIM_Cubic, tangentMode=RCTM_User, time=1, value=1}
        ]}}})
    ScaleA=true
    ScaleRGB=true
   ) {module=ScaleColor001}
  ]
  rendererClasses [&NiagaraRibbonRendererProperties]
  renderers [{rendererClass=&NiagaraRibbonRendererProperties}]
 }
 NE_VC_Boost_Greenwave {
  emitterUpdateScript [
   &EmitterState(
    Inactive Response=::NewEnumerator1
    Life Cycle Mode=::NewEnumerator1
    Loop Duration=0.1
    Recalculate Duration Each Loop=true
    UseLoopDelay=false
   )
   &VC_FilteredSocketCount()
   &VC_SpawnRibbon_v1(Number of Ribbons=$Emitter.PortCount, SpawnRate=400)
  ]
  particleSpawnScript [
   &VC_PaintedVariantsOverride(
    Black=<rgba 0.004 0.004 0.004 2>
    BurntSienna=<rgba 0.15 0.0404 0.0185 1>
    Cobalt=<rgba 0 0.0167 1 0.5>
    Crimson=<rgba 0.2 0 0 1>
    DefaultColor=&MakeCustomLinearColorFromBool(Bool=$User.bIsFrontendPreview, True Color=<rgba 0 0.5 0.0833 2>, False Color=<rgba 0 1.2 0.1831 1>)
    ForestGreen=<rgba 0 0.12 0 1.3>
    Grey=<rgba 0.06 0.06 0.06 1.5>
    Lime=<rgba 0.1333 0.2 0 2>
    Orange=<rgba 0.35 0.07 0 1>
    Pink=<rgba 0.3667 0.1 0.5 1>
    Purple=<rgba 0.0707 0.02 0.4 1.5>
    Saffron=<rgba 0.17 0.1417 0 1.7>
    SkyBlue=<rgba 0 0.1333 0.4 2>
    TitaniumWhite=<rgba 0.2 0.2 0.2 1.5>
   )
   &VC_InitializeRibbon_v1(
    Color=$Particles.Color
    Custom Offset=true
    Enable Skeletal Position=true
    Local Vector Offset=&AddVectorToPosition(
     Position=&MakeCustomVectorFromBool(Bool=$User.bIsFrontendPreview, True Vector=<100 0 0>, False Vector=<0 0 0>)
     Vector=$User.DTCarOffset
    )
    Offset Amount=1
    Ribbon Width=320
   )
   &AddVelocity(
    Velocity=&MakeCustomVectorFromBool(Bool=$User.bIsFrontendPreview, True Vector=<4000 0 0>, False Vector=<2500 0 0>)
    Velocity Speed Scale=-1
   ) {module=AddVelocity001}
  ]
  particleUpdateScript [
   &ParticleState()
   &InheritSourceMovement(
    Applied Force Scale=&Multiply_VectorByFloat(
     Vector=<1 1 1>
     Float=&MakeCustomFloatFromBool(Bool=$User.bIsFrontendPreview, True Float=0, False Float=1)
    )
    Applied Position Scale=&Multiply_VectorByFloat(
     Vector=<1 1 1>
     Float=&MakeCustomFloatFromBool(
      Bool=$User.bIsFrontendPreview
      True Float=0
      False Float=&FloatFromCurve(FloatCurve=DataInterface() {propertyValues {Curve {keys [{interpMode=RCIM_Cubic, tangentMode=RCTM_User, time=0, value=1}, {interpMode=RCIM_Cubic, tangentMode=RCTM_User, time=0.98, value=0}]}}})
     )
    )
    ApplyForce=true
    ApplyForceFalloff=false
    ApplyPosition=true
    ApplyPositionFalloff=false
    Module.LimitForceSpeed=false
   )
   &Drag(Drag=0.2, Rotational Drag=10, Use Linear Drag=true, Use Rotational Drag=true)
   &SolveForcesAndVelocity(Clamp Velocity=false, Limit Acceleration=false)
   &VC_UpdateRibbon_v1(Enable Skeletal Position=true)
   &DynamicMaterialParameters(
    Index 0 Param 1=&MakeCustomFloatFromBool(Bool=$User.bIsFrontendPreview, True Float=1, False Float=1.5)
    Index 0 Param 2=1
    Index 0 Param 3=&MakeCustomFloatFromBool(
     Bool=$User.bIsFrontendPreview
     True Float=&MakeCustomFloatFromBool(Bool=$User.bIsPainted, True Float=1, False Float=1)
     False Float=&MakeCustomFloatFromBool(Bool=$User.bIsPainted, True Float=0.5, False Float=0.25)
    )
    Index 0 Param 4=&MakeCustomFloatFromBool(Bool=$User.bIsPainted, True Float=1, False Float=2)
    Module.Index0Param0Write=true
    Module.Index0Param1Write=true
    Module.Index0Param2Write=true
    Module.Index0Param3Write=true
   )
   &ScaleRibbonWidth(Ribbon Width Scale=&SelectFloatFromArray(Direct Array Index=$Particles.SpawnGroup, Float Selection Array=DataInterface() {propertyValues {FloatData [-1,1]}}))
   &ScaleColor(
    Scale Alpha=&FloatFromCurve(FloatCurve=DataInterface() {propertyValues {Curve {keys [
         {arriveTangent=0.5243, interpMode=RCIM_Cubic, leaveTangent=0.5243, tangentMode=RCTM_User, time=0, value=0}
         {interpMode=RCIM_Cubic, tangentMode=RCTM_User, time=1, value=1}
        ]}}})
    ScaleA=true
    ScaleRGB=true
   )
  ]
  rendererClasses [&NiagaraRibbonRendererProperties]
  renderers [{rendererClass=&NiagaraRibbonRendererProperties}]
 }
 NE_VC_Boost_KOIFISH {
  emitterUpdateScript [
   &EmitterState(Inactive Response=::NewEnumerator1, Life Cycle Mode=::NewEnumerator1, Loop Duration=1, UseLoopDelay=false)
   &VC_FilteredSocketCount()
   &SpawnRate(
    SpawnRate=&MakeCustomFloatFromBool(Bool=$User.bIsFrontendPreview, True Float=10, False Float=50)
    Use Spawn Probability=false
   )
  ]
  particleSpawnScript [
   &VC_PaintedVariantsOverride(
    Black=<rgba 0.18 0.18 0.18 1>
    BurntSienna=<rgba 0.35 0.2042 0 1>
    Cobalt=<rgba 0 0.25 1 1>
    Crimson=<rgba 0.5 0 0.0083 1>
    DefaultColor=<rgba 0.05 1 0.4458 1>
    ForestGreen=<rgba 0.1 0.4 0 1>
    Grey=<rgba 0.4 0.4 0.4 1>
    Lime=<rgba 0.6333 1 0 1>
    Orange=<rgba 1 0.4167 0 1>
    Pink=<rgba 1.8 0 1.1455 1>
    Purple=<rgba 0.2667 0 0.4 1>
    Saffron=<rgba 0.9167 1 0 1>
    SkyBlue=<rgba 0 0.525 0.9 1>
    TitaniumWhite=<rgba 0.9 0.9 0.9 1>
   )
   &InitializeParticle(
    Color=$Particles.Color
    Lifetime=$Emitter.CurrentLoopDuration
    Mesh Scale=&MakeCustomVectorFromBool(
     Bool=$User.bIsFrontendPreview
     True Vector=&Multiply_VectorByFloat(Vector=<1 1 1>, Float=4)
     False Vector=&Multiply_VectorByFloat(Vector=<0.75 1 1>, Float=4)
    )
    Mesh Scale Mode=::NewEnumerator4
    Sprite Rotation Mode=::NewEnumerator2
    Sprite Size Mode=::NewEnumerator3
    Uniform Sprite Size=100
    UsePositionOffset=false
    Write Material Random=true
   )
   &PartitionParticles(Number Of Partitions=$Emitter.PortCount)
   &SkeletalMeshLocation(
    Bone / Socket Index=$Particles.PartitionParticles.Partition
    Bone / Socket Sampling Mode=::NewEnumerator5
    MaskSpawnGroup=false
    Mesh Sampling Type=::NewEnumerator6
    Sampled Position Offset=&AddVectorToPosition(Position=<-30 0 0>, Vector=$User.DTCarOffset)
    Skeletal Mesh=$System.SkeletalMesh
   )
   &AddVelocity(Velocity=<-1500 0 0>) {enabled=false}
  ]
  particleUpdateScript [
   &ParticleState()
   &SolveForcesAndVelocity(Clamp Velocity=false, Limit Acceleration=false)
   &ScaleColor(
    Scale RGB=&Multiply_VectorByFloat(
     Vector=<1 1 1>
     Float=&MakeCustomFloatFromBool(
      Bool=$User.bIsFrontendPreview
      True Float=1
      False Float=&MakeCustomFloatFromBool(Bool=$User.bIsPainted, True Float=1, False Float=0.5)
     )
    )
    ScaleA=true
    ScaleRGB=true
   )
  ]
  rendererClasses [&NiagaraMeshRendererProperties, &NiagaraSpriteRendererProperties]
  renderers [
   {rendererClass=&NiagaraMeshRendererProperties, rendererIndex=0}
   {rendererClass=&NiagaraSpriteRendererProperties, rendererIndex=1}
  ]
 }
 NE_VC_Boost_Sprite_Parent {
  emitterUpdateScript [
   &EmitterState(Life Cycle Mode=::NewEnumerator1, Loop Duration=1, UseLoopDelay=false)
   &VC_FilteredSocketCount()
   &SpawnRate(
    SpawnRate=&MakeCustomFloatFromBool(Bool=$User.bIsFrontendPreview, True Float=120, False Float=80)
    Use Spawn Probability=false
   )
  ]
  particleSpawnScript [
   &VC_PaintedVariantsOverride(
    Black=<rgba 0.02 0.02 0.02 25>
    BurntSienna=<rgba 0.2 0.0831 0.0247 0.05>
    Cobalt=<rgba 0.016 0.0305 0.8 1.5>
    Crimson=<rgba 0.2 0.0167 0 0.05>
    DefaultColor=&MakeCustomLinearColorFromBool(Bool=$User.bIsFrontendPreview, True Color=<rgba 0.02 2 0.185 5>, False Color=<rgba 0.1261 1 0.3235 1>)
    ForestGreen=<rgba 0.035 0.35 0.0509 0.3>
    Grey=<rgba 0.1 0.1 0.1 1>
    Lime=<rgba 0.1333 0.2 0 0.1>
    Orange=<rgba 0.4 0.1 0 0.02>
    Pink=<rgba 0.195 0.09 0.3 0.07>
    Purple=<rgba 0.0625 0 0.75 0.1>
    Saffron=<rgba 0.25 0.2292 0 0.02>
    SkyBlue=<rgba 0.0536 0.2146 0.5365 0.25>
    TitaniumWhite=<rgba 0.4 0.4 0.4 0.01>
   )
   &InitializeParticle(
    Color=$Particles.Color
    Lifetime Max=&MakeCustomFloatFromBool(Bool=$User.bIsFrontendPreview, True Float=0.9, False Float=1.5)
    Lifetime Min=0.7
    Lifetime Mode=::NewEnumerator1
    Sprite Rotation Mode=::NewEnumerator2
    Sprite Size Mode=::NewEnumerator1
    Uniform Sprite Size Max=40
    Uniform Sprite Size Min=10
    UsePositionOffset=false
    Write Material Random=true
   )
   &PartitionParticles(Number Of Partitions=$Emitter.PortCount)
   &SkeletalMeshLocation(
    Bone / Socket Index=$Particles.PartitionParticles.Partition
    Bone / Socket Sampling Mode=::NewEnumerator5
    MaskSpawnGroup=false
    Mesh Sampling Type=::NewEnumerator6
    Sampled Position Offset=&AddVectorToPosition(Position=<-20 0 0>, Vector=$User.DTCarOffset)
    Skeletal Mesh=$System.SkeletalMesh
   )
   &AddVelocity(
    Velocity=&RandomRangeVector(Minimum=<1500 250 -350>, Maximum=<2000 -250 350>)
    Velocity Speed Scale=&MakeCustomFloatFromBool(Bool=$User.bIsFrontendPreview, True Float=-1, False Float=-1)
   )
   &ShapeLocation(MaskSpawnGroup=false, Override Seed=false, Sphere Radius=45, UseOwnerScale=true)
  ]
  particleUpdateScript [
   &ParticleState()
   &ScaleSpriteSize(Uniform Curve Sprite Scale=DataInterface() {propertyValues {Curve {keys [{interpMode=RCIM_Cubic, tangentMode=RCTM_User, time=0, value=1}, {interpMode=RCIM_Cubic, tangentMode=RCTM_User, time=1, value=0}]}}})
   &VortexForce(
    Origin Pull Amount:NiagaraFloat=&MakeCustomFloatFromBool(Bool=$User.bIsFrontendPreview, True Float=15000, False Float=10000)
    UseOriginOffset=true
    Vortex Axis=&RandomRangeVector(Minimum=<-1 0 0>, Maximum=<1 0 0>, Evaluation Type=::NewEnumerator1)
    Vortex Axis Coordinate Space:ENiagaraCoordinateSpace=::Local
    Vortex Force Amount=&MakeCustomFloatFromBool(Bool=$User.bIsFrontendPreview, True Float=5000, False Float=7500)
    Vortex Origin:NiagaraPosition=$Particles.Initial.Position
    Vortex Origin Offset:Vector3f=&RandomRangeVector(Minimum=<-100 0 0>, Maximum=<-400 0 0>, Evaluation Type=::NewEnumerator1)
    Vortex Origin Offset Coordinate Space:ENiagaraCoordinateSpace=::Local
   )
   &InheritSourceMovement(
    Applied Position Scale=&Multiply_VectorByFloat(
     Vector=<1 1 1>
     Float=&FloatFromCurve(FloatCurve=DataInterface() {propertyValues {Curve {keys [{interpMode=RCIM_Cubic, tangentMode=RCTM_User, time=0.3, value=1}, {interpMode=RCIM_Cubic, tangentMode=RCTM_User, time=1, value=0}]}}})
    )
    ApplyForce=false
    ApplyPosition=true
    ApplyPositionFalloff=false
   )
   &SolveForcesAndVelocity(Clamp Velocity=false, Limit Acceleration=false)
   &ScaleColor(
    Scale Alpha=&FloatFromCurve(FloatCurve=DataInterface() {propertyValues {Curve {keys [<at 0 1>, {interpMode=RCIM_Cubic, tangentMode=RCTM_User, time=0.8, value=1}, {interpMode=RCIM_Cubic, tangentMode=RCTM_User, time=1, value=0}]}}})
    Scale RGB=&Multiply_VectorByFloat(Vector=<1 1 1>, Float=&MakeCustomFloatFromBool(Bool=$User.bIsPainted, True Float=5, False Float=2))
    ScaleA=true
    ScaleRGB=true
   )
   &DynamicMaterialParameters(
    Index 0 Param 1=12
    Index 0 Param 2=1
    Module.Index0Param0Write=true
    Module.Index0Param1Write=true
    Module.Index0Param2Write=true
    Module.Index0Param3Write=true
   )
  ]
  rendererClasses [&NiagaraSpriteRendererProperties]
  renderers [{rendererClass=&NiagaraSpriteRendererProperties}]
 }
}
summary {
 systemName=NS_Boost_Koi
 emitters [
  {
   emitterName=NE_VC_Boost_KOIFISH
   rendererClasses [&NiagaraMeshRendererProperties, &NiagaraSpriteRendererProperties]
  }
  {emitterName=NE_VC_Boost_Aura, rendererClasses [&NiagaraRibbonRendererProperties]}
  {emitterName=NE_VC_Boost_Greenwave, rendererClasses [&NiagaraRibbonRendererProperties]}
  {emitterName=NE_VC_Boost_Sprite_Parent, rendererClasses [&NiagaraSpriteRendererProperties]}
  {emitterName=NE_VC_Boost_BLUEWAVE001, rendererClasses [&NiagaraRibbonRendererProperties]}
 ]
 userVariables [
  User.bIsFrontendPreview {type=NiagaraBool, defaultValue=true}
  User.bIsPainted {type=NiagaraBool}
  User.ColorVariantID {type=NiagaraInt32}
  User.DTCarOffset {type=Vector3f, description="", defaultValue {struct=Vector3f, value=<0 0 0>}}
  User.PaintedColor {type=LinearColor, description="", defaultValue {struct=LinearColor, value=<rgba 1 1 1 1>}}
  User.PaintedScalar {type=NiagaraFloat, defaultValue=1, description=""}
  User.PortCount {type=NiagaraInt32, defaultValue=2}
 ]
}
