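# Translucent unlit, two-sided Niagara ribbon material for an aura trail in the Koi vehicle boost.
# Dynamic parameters drive the aura texture's pan speed, offset and final strength; that texture is multiplied by two scrolling foam-gradient noise samples.
# Power fades along U taper both ribbon ends, then particle alpha, camera depth fade and saturation shape the mask.
# Particle color feeds a shared emissive-adjustment function.
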Material /VehicleCosmetics/Boosts/Boost_Koi/Materials/M_Aura_03
graph {
 properties {
  bUsedWithNiagaraRibbons=true
  blendMode=BLEND_Translucent
  pixelDepthOffsetMode=PDOM_Legacy
  shadingModel=MSM_Unlit
  twoSided=true
  usedShadingModels=Unlit
 }
 nodes [
  n1 = DynamicParameter() {paramNames [speed,offchange,opacity,auraopacity]}
  n2 = ScalarParameter() {parameterName=append param}
  n3 = Panner(Coordinate=Add(A=AppendVector(A=@n1.offchange, B=@n2), B=TextureCoordinate()), Time=Multiply(A=@n1.speed, B=Time())) {speedX=-1}
  MaterialExpressionVertexColor_2 = ParticleColor()
  n4 = Multiply(
   A=Multiply(A=TextureSample(UVs=@n3) {texture=/VehicleCosmetics/Boosts/Boost_Koi/Textures/T_VCVFX_Boost_Koi_AuraTrail02}.RGB, B=Constant() {r=10})
   B=@MaterialExpressionVertexColor_2.A
  )
  n5 = TextureCoordinate()
  n6 = Multiply(
   A=TextureCoordinate()
   B=AppendVector(A=ScalarParameter() {defaultValue=1.5, parameterName=X}, B=ScalarParameter() {defaultValue=3, parameterName=Y})
  )
  n7 = TextureSample(UVs=Panner(Coordinate=TextureCoordinate() {vTiling=1.5}) {speedY=2.5}) {texture=/DealershipSharedLibrary/FX/Textures/TexturePackage/T_VCVFX_FoamGradient01_RGB}
  n8 = Multiply(
   A=TextureSample(UVs=Panner(Coordinate=@n6) {speedX=-1.3}) {texture=&T_VCVFX_FoamGradient01_RGB}.G
   B=@n7.R
  )
  n9 = Constant() {r=200}
  n10 = Power(Base=@n7.R, Exp=Constant() {r=0.6})
  n11 = Constant()
  n12 = Constant() {r=0.335}
  n13 = Multiply(
   A=Power(Base=ComponentMask(None=@n5) {r=true}, Exp=ScalarParameter() {defaultValue=2, parameterName=power fade})
   B=Add(A=Multiply(A=@n8, B=@n9), B=Clamp(None=@n10, Min=@n11, Max=@n12))
  )
  n14 = OneMinus(None=ComponentMask(None=TextureCoordinate()) {r=true})
  n15 = ScalarParameter() {defaultValue=2, parameterName=power fade one minus}
  n16 = Multiply(A=Multiply(A=Multiply(A=@n4, B=@n13), B=Power(Base=@n14, Exp=@n15)), B=Constant() {r=0.5})
  n17 = MaterialFunctionCall(Fade Length=Constant() {r=100}, Fade Offset=Constant() {r=50}) {materialFunction=&CameraDepthFade}
  MaterialExpressionConstantClamp_2 = Saturate(None=Multiply(A=Multiply(A=@n16, B=@n17.Result), B=@n1.auraopacity))
  MaterialFunctionCall(Emissive=@MaterialExpressionVertexColor_2.RGB) {materialFunction=&MF_FX_EmissiveAdjustments}
 ]
}
