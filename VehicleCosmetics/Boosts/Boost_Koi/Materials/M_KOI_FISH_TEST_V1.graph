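# Test unlit masked material for the koi fish in the Koi vehicle boost effect, set up for Niagara mesh particles.
# Body, scale and panning textures are tinted by a time-oscillating hue shift of particle color, then shaded by a view-angle rim and normal-map-derived mask.
# A noise-driven dissolve adds a glowing edge band before the shared emissive adjustment function.
# A morph-target function takes offset and normal textures for timed vertex animation.
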Material /VehicleCosmetics/Boosts/Boost_Koi/Materials/M_KOI_FISH_TEST_V1
graph {
 properties {
  bTangentSpaceNormal=false
  bUsedWithNiagaraMeshParticles=true
  blendMode=BLEND_Masked
  maxWorldPositionOffsetDisplacement=50
  numCustomizedUVs=4
  pixelDepthOffsetMode=PDOM_Legacy
  shadingModel=MSM_Unlit
  usedShadingModels=Unlit
 }
 nodes [
  MaterialFunctionCall!(
   Morph Animation=MaterialFunctionCall(Speed=ScalarParameter() {defaultValue=0.5}) {materialFunction=&TimeWithSpeedVariable}.Time
   Morph Normal=TextureObject() {samplerType=SAMPLERTYPE_LinearColor, texture=/VehicleCosmetics/Boosts/Boost_Koi/Textures/T_VCVFX_Boost_Koi_Koi_normals_283}
   Morph Texture=TextureObject() {samplerType=SAMPLERTYPE_LinearColor, texture=/VehicleCosmetics/Boosts/Boost_Koi/Textures/T_VCVFX_Boost_Koi_Koi_offsets_283}
   Number of Morph Targets
   0-1 Animation Value?
   Morph Target UV
  ) {materialFunction=/Engine/Functions/Engine_MaterialFunctions02/MAXScripts/MS_VertexAnimationTools_MorphTargets}
  n1 = Sine(None=Time())
  n2 = Constant() {r=1}
  n3 = MaterialFunctionCall(
   Hue Shift Percentage=Divide(A=Add(A=@n1, B=@n2), B=ScalarParameter() {defaultValue=-25, parameterName=SecondColor})
   Texture=ParticleColor().RGB
  ) {materialFunction=&HueShift}
  MaterialExpressionConstantClamp_18 = Saturate(None=Multiply(A=TextureSample() {texture=/VehicleCosmetics/Boosts/Boost_Koi/Textures/T_VCVFX_Boost_Koi_GoldKoi03}.RGB, B=Constant() {r=100}))
  n4 = Panner() {speedX=1}
  n5 = Multiply(
   A=Multiply(A=TextureSample(UVs=@n4) {texture=/VehicleCosmetics/Boosts/Boost_Koi/Textures/T_VCVFX_Boost_Koi_Panner02}.RGB) {constB=2}
   B=Multiply(A=TextureSample() {texture=/VehicleCosmetics/Boosts/Boost_Koi/Textures/T_VCVFX_Boost_Koi_Esacamas04}.RGB, B=Constant() {r=10})
  )
  n6 = TextureSample() {texture=/VehicleCosmetics/Boosts/Boost_Koi/Textures/T_VCVFX_Boost_Koi_BodyKoi04}
  MaterialExpressionConstantClamp_17 = Saturate(None=Multiply(A=Multiply(A=@n6.RGB, B=@n3.Result), B=Constant() {r=2}))
  n7 = LinearInterpolate(
   A=Multiply(A=@n3.Result, B=Add(A=@MaterialExpressionConstantClamp_18, B=@n5))
   B=@MaterialExpressionConstantClamp_17
   Alpha=TextureSample() {texture=/VehicleCosmetics/Boosts/Boost_Koi/Textures/T_VCVFX_Boost_Koi_KoiAlpha02}.RGB
  )
  n8 = Constant3Vector() {constant=<rgba 0.0994 0.0994 0.0994 0>}
  n9 = ComponentMask(None=Transform(None=CameraVectorWS()) {transformSourceType=TRANSFORMSOURCE_World, transformType=TRANSFORM_Tangent}) {b=true}
  MaterialExpressionConstantClamp_16 = Saturate(None=ConstantBiasScale(None=OneMinus(None=@n9)) {bias=-0.15, scale=2})
  n10 = TextureSample() {texture=/VehicleCosmetics/Boosts/Boost_Koi/Textures/T_VCVFX_Boost_Koi_KoiNormal}
  MaterialExpressionConstantClamp_15 = Saturate(None=ConstantBiasScale(None=Desaturation(None=@n10.RGB)) {bias=-0.1, scale=10})
  n11 = Multiply(
   A=ScalarParameter() {defaultValue=3, parameterName=Britghtness}
   B=Multiply(A=LinearInterpolate(A=@n7, B=@n8, Alpha=@MaterialExpressionConstantClamp_16), B=@MaterialExpressionConstantClamp_15)
  )
  n12 = Multiply(A=@n3.Result, B=Constant() {r=2})
  n13 = Multiply(A=TextureSample() {texture=/DealershipSharedLibrary/FX/Textures/Noise/T_VCVFX_Noise1_A}.RGB, B=Constant() {r=1})
  n14 = Constant()
  n15 = Constant() {r=1}
  n16 = Power(Base=ScalarParameter() {defaultValue=2, parameterName=Strange}, Exp=Constant() {r=3})
  n17 = OneMinus(None=@n16)
  n18 = OneMinus(None=ScalarParameter() {defaultValue=1, parameterName=Dissolve})
  n19 = Multiply(A=ScalarParameter() {defaultValue=1, parameterName=Remap_Dissolve}, B=Constant() {r=12})
  n20 = OneMinus(None=@n19)
  n21 = Add(
   A=ComponentMask(None=MaterialFunctionCall(Input=@n13, Input Low=@n14, Input High=@n15, Target Low=@n17, Target High=@n16) {materialFunction=&RemapValueRange}.Result) {r=true}
   B=ComponentMask(None=MaterialFunctionCall(Input=@n18, Input Low=@n14, Input High=@n15, Target Low=@n20, Target High=@n19) {materialFunction=&RemapValueRange}.Result) {r=true}
  )
  n22 = ScalarParameter() {defaultValue=1, parameterName=Line}
  n23 = OneMinus(None=ScalarParameter() {defaultValue=1, parameterName=Line})
  n24 = Constant() {r=1}
  MaterialExpressionConstantClamp_19 = Saturate(None=@n21)
  n25 = Constant() {r=1}
  n26 = Multiply(
   A=If(A=@n21, B=Multiply(A=@n22, B=@n23), "A > B"=@n24, "A == B"=@n24, "A < B"=Constant())
   B=If(A=Constant() {r=0.8}, B=@MaterialExpressionConstantClamp_19, "A > B"=@n25, "A == B"=@n25, "A < B"=Constant())
  )
  MaterialFunctionCall(Emissive=Add(A=Multiply(A=@n3.Result, B=@n11), B=Multiply(A=@n12, B=@n26))) {materialFunction=&MF_FX_EmissiveAdjustments}
  VectorParameter() {parameterName=KoiColor, defaultValue=<rgba 0.0196 0.902 0.4392 1>}
 ]
}
