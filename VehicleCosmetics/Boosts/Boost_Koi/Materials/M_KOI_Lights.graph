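# Translucent unlit Niagara sprite material for lights in the Koi vehicle boost effect, drawing a soft round glow without textures.
# A dot product of centered UVs builds a clamped radial falloff, which is tinted by particle color and squared.
# That result feeds the MF_FX_EmissiveAdjustments function and is also multiplied by particle alpha and a dynamic Opacity parameter.
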Material /VehicleCosmetics/Boosts/Boost_Koi/Materials/M_KOI_Lights
graph {
 properties {bUsedWithNiagaraSprites=true, blendMode=BLEND_Translucent, pixelDepthOffsetMode=PDOM_Legacy, shadingModel=MSM_Unlit, usedShadingModels=Unlit}
 nodes [
  MaterialExpressionVertexColor_8 = ParticleColor()
  n1 = Add(A=TextureCoordinate(), B=Constant() {r=-0.5})
  n2 = OneMinus(None=DotProduct(A=@n1, B=@n1))
  n3 = Constant() {r=0.8}
  MaterialExpressionConstantClamp_8 = Saturate(None=Multiply(A=Subtract(A=@n2, B=@n3), B=Constant() {r=5}))
  n4 = Power(Base=Multiply(A=@MaterialExpressionVertexColor_8.RGB, B=@MaterialExpressionConstantClamp_8), Exp=Constant() {r=2})
  MaterialFunctionCall(Emissive=@n4) {materialFunction=&MF_FX_EmissiveAdjustments}
  Multiply(A=@n4, B=Multiply(A=@MaterialExpressionVertexColor_8.A, B=DynamicParameter() {paramNames [Opacity,Param2,Param3,Param4]}.Opacity))
 ]
}
