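# Unlit translucent Niagara ribbon material for the Koi boost's smoke trail, tinted by particle color.
# Smoke and outline textures pan along U, with dynamic parameters driving speed, offset, opacity and outline intensity.
# Power ramps at both ribbon ends, particle alpha and camera depth fade shape the saturated masks, while tinted outlines feed an emissive-adjustment function.
# The graph also holds parallel variant chains, including an aura trail modulated by panning foam noise.
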Material /VehicleCosmetics/Boosts/Boost_Koi/Materials/M_Smoke_Trail_Blue
graph {
 properties {bUsedWithNiagaraRibbons=true, blendMode=BLEND_Translucent, pixelDepthOffsetMode=PDOM_Legacy, shadingModel=MSM_Unlit, usedShadingModels=Unlit}
 nodes [
  n1 = DynamicParameter() {paramNames [speed, offchange, opacity, Outlines Intensity]}
  n2 = ParticleColor()
  n3 = Constant() {r=2}
  n4 = DynamicParameter() {paramNames [speed, offchange, opacity, Outlines Intensity]}
  n5 = AppendVector(A=@n4.offchange, B=ScalarParameter() {parameterName=append param})
  n6 = StaticSwitchParameter(True=TextureCoordinate() {uTiling=1.5, vTiling=-1}, False=TextureCoordinate() {uTiling=1.5}) {parameterName=Switch}
  n7 = Time()
  n8 = TextureSample(UVs=Panner(Coordinate=Add(A=@n5, B=@n6), Time=Multiply(A=@n4.speed, B=@n7)) {speedX=-1}) {texture=/VehicleCosmetics/Boosts/Boost_Koi/Textures/T_VCVFX_Boost_Koi_TrailSmokeOutline2_1}
  n9 = Multiply(A=@n2.RGB, B=Multiply(A=Multiply(A=@n2.RGB, B=@n3), B=Multiply(A=@n8.RGB, B=@n1.Outlines Intensity)))
  n10 = Add(
   A=AppendVector(A=@n1.offchange, B=ScalarParameter() {parameterName=append param})
   B=StaticSwitchParameter(True=TextureCoordinate() {uTiling=1.5, vTiling=-1}, False=TextureCoordinate() {uTiling=1.5}) {parameterName=Switch}
  )
  n11 = Multiply(A=@n1.speed, B=Time())
  n12 = Multiply(
   A=TextureSample(UVs=Panner(Coordinate=@n10, Time=@n11) {speedX=-1}) {texture=/VehicleCosmetics/Boosts/Boost_Koi/Textures/T_VCVFX_Boost_Koi_TrailSmoke01_1}.RGB
   B=Constant() {r=10}
  )
  n13 = TextureCoordinate()
  n14 = Multiply(
   A=Multiply(A=Add(A=@n9, B=@n12), B=@n2.A)
   B=Power(Base=ComponentMask(None=@n13) {r=true}, Exp=ScalarParameter() {defaultValue=1.5, parameterName=power fade})
  )
  n15 = TextureCoordinate()
  n16 = Power(Base=OneMinus(None=ComponentMask(None=@n15) {r=true}), Exp=ScalarParameter() {defaultValue=2, parameterName=power fade one minus})
  n17 = Multiply(
   A=Multiply(A=Multiply(A=@n14, B=@n16), B=Constant() {r=1})
   B=MaterialFunctionCall(Fade Length=Constant() {r=100}, Fade Offset=Constant() {r=50}) {materialFunction=&CameraDepthFade}.Result
  )
  Add(A=Saturate(None=Multiply(A=@n1.opacity, B=@n17))) {constB=0}
  MaterialExpressionVertexColor_4 = ParticleColor()
  n18 = AppendVector(A=@MaterialExpressionVertexColor_4.RGB, B=Constant() {r=1})
  n19 = DynamicParameter() {paramNames [speed, offchange, opacity, Outlines Intensity]}
  n20 = AppendVector(A=@n19.offchange, B=ScalarParameter() {parameterName=append param})
  n21 = StaticSwitchParameter(True=TextureCoordinate() {uTiling=1.5, vTiling=-1}, False=TextureCoordinate() {uTiling=1.5}) {parameterName=Switch}
  n22 = Time()
  n23 = TextureSample(UVs=Panner(Coordinate=Add(A=@n20, B=@n21), Time=Multiply(A=@n19.speed, B=@n22)) {speedX=-1}) {texture=&T_VCVFX_Boost_Koi_TrailSmokeOutline2_1}
  n24 = Constant() {r=5}
  n25 = Multiply(
   A=Multiply(A=@MaterialExpressionVertexColor_4.RGBA, B=Constant() {r=10})
   B=MaterialFunctionCall!(In=Multiply(A=@n23.RGB, B=@n24), Ramp Width=Constant() {r=0.1}) {materialFunction=/Engine/Functions/Engine_MaterialFunctions01/Math/SmoothCeil}.Result
  )
  Add(
   A=Add(A=@n18, B=@n25)
   B=Multiply(A=@n18, B=TextureSample() {texture=&T_VCVFX_Boost_Koi_TrailSmokeOutline2_1}.RGBA)
  )
  n26 = DynamicParameter() {paramNames [speed, offchange, opacity, Outlines Intensity]}
  n27 = ScalarParameter() {parameterName=append param}
  n28 = Panner(Coordinate=Add(A=AppendVector(A=@n26.offchange, B=@n27), B=TextureCoordinate()), Time=Multiply(A=@n26.speed, B=Time())) {speedX=-1}
  n29 = Multiply(
   A=Multiply(A=TextureSample(UVs=@n28) {texture=/VehicleCosmetics/Boosts/Boost_Koi/Textures/T_VCVFX_Boost_Koi_AuraTrail02}.RGB, B=Constant() {r=10})
   B=@n2.A
  )
  n30 = TextureCoordinate()
  n31 = Multiply(
   A=TextureCoordinate()
   B=AppendVector(A=ScalarParameter() {defaultValue=1.5, parameterName=X}, B=ScalarParameter() {defaultValue=3, parameterName=Y})
  )
  n32 = TextureSample(UVs=Panner(Coordinate=TextureCoordinate() {vTiling=1.5}) {speedY=2.5}) {texture=/DealershipSharedLibrary/FX/Textures/TexturePackage/T_VCVFX_FoamGradient01_RGB}
  n33 = Multiply(
   A=TextureSample(UVs=Panner(Coordinate=@n31) {speedX=-1.3}) {texture=&T_VCVFX_FoamGradient01_RGB}.G
   B=@n32.R
  )
  n34 = Constant() {r=200}
  n35 = Power(Base=@n32.R, Exp=Constant() {r=0.6})
  n36 = Constant()
  n37 = Constant() {r=0.335}
  n38 = Multiply(
   A=Power(Base=ComponentMask(None=@n30) {r=true}, Exp=ScalarParameter() {defaultValue=1.5, parameterName=power fade})
   B=Add(A=Multiply(A=@n33, B=@n34), B=Clamp(None=@n35, Min=@n36, Max=@n37))
  )
  n39 = OneMinus(None=ComponentMask(None=TextureCoordinate()) {r=true})
  n40 = ScalarParameter() {defaultValue=2, parameterName=power fade one minus}
  n41 = Multiply(A=Multiply(A=Multiply(A=@n29, B=@n38), B=Power(Base=@n39, Exp=@n40)), B=Constant() {r=0.5})
  n42 = MaterialFunctionCall(Fade Length=Constant() {r=100}, Fade Offset=Constant() {r=50}) {materialFunction=&CameraDepthFade}
  MaterialExpressionConstantClamp_2 = Saturate(None=Multiply(A=Multiply(A=@n41, B=@n42.Result), B=@n26.Outlines Intensity))
  n43 = DynamicParameter() {paramNames [speed, offchange, opacity, Outlines Intensity]}
  n44 = AppendVector(A=@n43.offchange, B=ScalarParameter() {parameterName=append param})
  n45 = StaticSwitchParameter(True=TextureCoordinate() {uTiling=1.5, vTiling=-1}, False=TextureCoordinate() {uTiling=1.5}) {parameterName=Switch}
  n46 = Time()
  n47 = TextureSample(UVs=Panner(Coordinate=Add(A=@n44, B=@n45), Time=Multiply(A=@n43.speed, B=@n46)) {speedX=-1}) {texture=&T_VCVFX_Boost_Koi_TrailSmoke01_1}
  n48 = Constant() {r=20}
  n49 = Multiply(A=Add(A=@n25, B=Multiply(A=@n47.RGBA, B=@n48)), B=@MaterialExpressionVertexColor_4.A)
  n50 = Power(Base=ComponentMask(None=TextureCoordinate()) {r=true}, Exp=ScalarParameter() {defaultValue=1.5, parameterName=power fade})
  n51 = OneMinus(None=ComponentMask(None=TextureCoordinate()) {r=true})
  n52 = ScalarParameter() {defaultValue=2, parameterName=power fade one minus}
  n53 = Multiply(A=Multiply(A=Multiply(A=@n49, B=@n50), B=Power(Base=@n51, Exp=@n52)), B=Constant() {r=1})
  n54 = MaterialFunctionCall(Fade Length=Constant() {r=100}, Fade Offset=Constant() {r=50}) {materialFunction=&CameraDepthFade}
  MaterialExpressionConstantClamp_4 = Saturate(None=Multiply(A=@n43.opacity, B=Multiply(A=@n53, B=@n54.Result)))
  Constant() {r=5}
  MaterialFunctionCall(Emissive=Add(A=@n2.RGB, B=@n9)) {materialFunction=&MF_FX_EmissiveAdjustments}
 ]
}
