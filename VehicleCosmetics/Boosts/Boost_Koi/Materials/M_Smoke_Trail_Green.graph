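# Translucent unlit Niagara ribbon material for the green smoke trail of the Koi boost, tinted by particle color.
# Smoke and outline textures pan along U, with speed, offset and outline intensity driven by dynamic parameters and a switch flipping V.
# Particle alpha and two power curves on U fade their sum at both ends, then dynamic opacity scales and saturates it.
# Tinted outlines also feed an emissive adjustment function; a camera depth fade branch multiplies the faded mask.
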Material /VehicleCosmetics/Boosts/Boost_Koi/Materials/M_Smoke_Trail_Green
graph {
 properties {bUsedWithNiagaraRibbons=true, blendMode=BLEND_Translucent, pixelDepthOffsetMode=PDOM_Legacy, shadingModel=MSM_Unlit, usedShadingModels=Unlit}
 nodes [
  n1 = DynamicParameter() {paramNames [speed, offchange, opacity, Outlines Intensity]}
  MaterialExpressionVertexColor_4 = ParticleColor()
  n2 = DynamicParameter() {paramNames [speed, offchange, opacity, Outlines Intensity]}
  n3 = Add(
   A=AppendVector(A=@n2.offchange, B=ScalarParameter() {parameterName=append param})
   B=StaticSwitchParameter(True=TextureCoordinate() {uTiling=1.4, vTiling=-1}, False=TextureCoordinate() {uTiling=1.4}) {parameterName=Switch}
  )
  n4 = Multiply(A=@n2.speed, B=Time())
  n5 = Multiply(
   A=TextureSample(UVs=Panner(Coordinate=@n3, Time=@n4) {speedX=-1}) {texture=/VehicleCosmetics/Boosts/Boost_Koi/Textures/T_VCVFX_Boost_Koi_TrailSmokeOutline4}.RGB
   B=@n1.Outlines Intensity
  )
  n6 = Multiply(A=Multiply(A=@MaterialExpressionVertexColor_4.RGB, B=Constant() {r=10}), B=@n5)
  n7 = AppendVector(A=@n1.offchange, B=ScalarParameter() {parameterName=append param})
  n8 = StaticSwitchParameter(True=TextureCoordinate() {uTiling=1.4, vTiling=-1}, False=TextureCoordinate() {uTiling=1.4}) {parameterName=Switch}
  n9 = Time()
  n10 = TextureSample(UVs=Panner(Coordinate=Add(A=@n7, B=@n8), Time=Multiply(A=@n1.speed, B=@n9)) {speedX=-1}) {texture=/VehicleCosmetics/Boosts/Boost_Koi/Textures/T_VCVFX_Boost_Koi_TrailSmoke06}
  n11 = Constant() {r=10}
  n12 = Multiply(A=Add(A=@n6, B=Multiply(A=@n10.RGB, B=@n11)), B=@MaterialExpressionVertexColor_4.A)
  n13 = Power(Base=ComponentMask(None=TextureCoordinate()) {r=true}, Exp=ScalarParameter() {defaultValue=1.5, parameterName=power fade})
  n14 = OneMinus(None=ComponentMask(None=TextureCoordinate()) {r=true})
  n15 = ScalarParameter() {defaultValue=2, parameterName=power fade one minus}
  n16 = Multiply(A=Multiply(A=Multiply(A=@n12, B=@n13), B=Power(Base=@n14, Exp=@n15)), B=Constant() {r=1})
  MaterialExpressionConstantClamp_4 = Saturate(None=Multiply(A=@n1.opacity, B=@n16))
  Constant() {r=5}
  MaterialFunctionCall!(In=@n5, Ramp Width=Constant() {r=0.01}) {materialFunction=/Engine/Functions/Engine_MaterialFunctions01/Math/SmoothCeil}
  MaterialFunctionCall(Emissive=Add(A=@MaterialExpressionVertexColor_4.RGB, B=@n6)) {materialFunction=&MF_FX_EmissiveAdjustments}
  Multiply(
   A=@n16
   B=MaterialFunctionCall(Fade Length=Constant() {r=100}, Fade Offset=Constant() {r=50}) {materialFunction=&CameraDepthFade}.Result
  )
 ]
}
